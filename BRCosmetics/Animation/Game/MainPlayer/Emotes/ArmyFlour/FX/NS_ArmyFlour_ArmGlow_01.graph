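# Niagara system for the arm glow in the ArmyFlour player emote: two short-lived, soft orange-gold sprite glows on the character.
# A single emitter burst-spawns both sprites, and each uses its spawn group as a bone/socket index to track User.SkeletalMesh every update.
# Color and alpha curves fade the glow in and out over its lifetime.
# Camera offset depends on platform quality level, and a ScaleMeshSize module runs despite sprite-only rendering.
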NiagaraSystem /BRCosmetics/Animation/Game/MainPlayer/Emotes/ArmyFlour/FX/NS_ArmyFlour_ArmGlow_01
dependencies {
 usedDataInterfaces [&NiagaraDataInterfacePlatformSet, &NiagaraDataInterfaceColorCurve, &NiagaraDataInterfaceCurve]
 usedDynamicInputs [&Vector2DFromFloat]
 usedModules [&SystemState]
}
emitters {FN_SourceGlows002 {
  emitterSpawnScript [&FN_PreserveSkydiveUserVariables() {module=PreserveUserVariables}]
  emitterUpdateScript [
   &EmitterState()
   &SpawnBurst_Instantaneous(Spawn Count=2, Use Spawn Probability=false, UseLoopCountLimit=false)
  ]
  particleSpawnScript [
   &FN_SetSpawnGroup(Number of Spawn Groups=2)
   &InitializeParticle(
    Lifetime=&UniformRangedFloat(Minimum=2, Maximum=2)
    Sprite Size Mode=::NewEnumerator3
    Uniform Sprite Size=90
    UsePositionOffset=false
    Write Material Random=true
   ) {module=InitializeParticle001}
   &CameraOffset(Camera Offset Amount=&FN_PickFloatDependOnPlatform(TargetPlatform=DataInterface() {propertyValues {Platforms {qualityLevelMask=-29}}}, TargetPlatformActive=20, TargetPlatformNoActive=10))
  ]
  particleUpdateScript [
   /SetVariables(Particles.Color=&MakeLinearColorFromVectorAndFloat(
     "Vector (RGB)"=&MakeVectorFromLinearColorRGB(Color=&ColorFromCurve(ColorCurve=DataInterface() {propertyValues {
         AlphaCurve {keys [<at -0.0023 1 RCIM_Cubic>,<at 0.9977 1 RCIM_Cubic>]}
         BlueCurve {keys [<at 0 0.1>,<at 0.341 0.15>,<at 1 0.1 RCIM_Cubic>]}
         GreenCurve {keys [<at 0 0.625>,<at 0.341 0.7167>,<at 1 0.625 RCIM_Cubic>]}
         RedCurve {keys [<at 0 1>,<at 0.341 1>,<at 1 1 RCIM_Cubic>]}
        }}))
     "Float (Alpha)"=&FloatFromCurve(FloatCurve=DataInterface() {propertyValues {Curve {keys [
          {arriveTangent=1.1784, interpMode=RCIM_Cubic, leaveTangent=1.1784, tangentMode=RCTM_User, time=0, value=0}
          {arriveTangent=0.002, interpMode=RCIM_Cubic, leaveTangent=0.002, tangentMode=RCTM_User, time=0.5, value=0.2}
          {arriveTangent=-1.0701, interpMode=RCIM_Cubic, leaveTangent=-1.0701, tangentMode=RCTM_User, time=1.0011, value=0.002}
         ]}}})
    )) {isSetParameters=true}
   &SkeletalMeshLocation(
    Bone / Socket Index=$Particles.SpawnGroup
    Bone Sampling Mode=::NewEnumerator4
    Inherit Velocity=false
    MaskSpawnGroup=false
    Sampled Position Offset=<0 0 -3>
    Skeletal Mesh=$User.SkeletalMesh
   )
   &ParticleState()
   &ScaleMeshSize(Scale Factor=&Multiply_Vector(
     A=<1 1 1>
     B=&VectorFromFloat(Value=&FloatFromCurve(FloatCurve=DataInterface() {propertyValues {Curve {keys [<at 0 0>,<at 1 1.5>]}}}, Scale Curve=2))
    ))
   &ScaleColor(
    Scale Alpha=&FloatFromCurve(FloatCurve=DataInterface() {propertyValues {Curve {keys [
         <at 0.0028 0.0306 RCIM_Cubic>
         {arriveTangent=-0.0307, interpMode=RCIM_Cubic, leaveTangent=-0.0307, time=0.6304, value=1.5546}
         <at 1 0 RCIM_Cubic>
        ]}}})
    Scale RGB=<0.4 0.3 0.1>
    ScaleA=true
    ScaleRGB=true
   )
  ]
  rendererClasses [&NiagaraSpriteRendererProperties]
  renderers [{rendererClass=&NiagaraSpriteRendererProperties}]
 }}
summary {
 systemName=NS_ArmyFlour_ArmGlow_01
 emitters [{emitterName=FN_SourceGlows002, rendererClasses [&NiagaraSpriteRendererProperties]}]
 userVariables [
  User.bIsFrontEndPreview {type=NiagaraBool}
  User.PlayerSpeed {type=NiagaraFloat, defaultValue=0, description=""}
  User.SkeletalMesh {
   type=NiagaraDataInterfaceSkeletalMesh
   defaultValue {
    struct=NiagaraExt_VariableValue_DataInterface
    value {dataInterface=&self:NiagaraDataInterfaceSkeletalMesh_1}
   }
  }
 ]
}
