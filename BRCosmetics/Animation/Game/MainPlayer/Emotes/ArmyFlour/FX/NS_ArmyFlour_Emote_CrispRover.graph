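# Niagara effect for the ArmyFlour emote (CrispRover) that puffs short-lived, dark brownish sprite clouds from two emitters named Steam.
# The emitters have different offsets and staggered short delays, spawning by rate and by movement distance with a random velocity that drag slows.
# Curves grow the sprites, fade and darken their color, and drive dynamic material parameters over life.
# A final color scale tints the particles by time of day.
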NiagaraSystem /BRCosmetics/Animation/Game/MainPlayer/Emotes/ArmyFlour/FX/NS_ArmyFlour_Emote_CrispRover
dependencies {
 usedDataInterfaces [&NiagaraDataInterfaceCurve, &NiagaraDataInterfaceColorCurve]
 usedDynamicInputs []
 usedModules [
  &SystemState
  &self:SystemSpawnScript.SystemScriptSource.SystemScriptGraph.NiagaraNodeAssignment_0.SetVariables_#1
  &self:SystemSpawnScript.SystemScriptSource.SystemScriptGraph.NiagaraNodeAssignment_1.SetVariables_#2
  &self:SystemSpawnScript.SystemScriptSource.SystemScriptGraph.NiagaraNodeAssignment_2.SetVariables_#3
  &self:SystemSpawnScript.SystemScriptSource.SystemScriptGraph.NiagaraNodeAssignment_5.SetVariables_#4
  &InterpolateOverTime
  &self:SystemSpawnScript.SystemScriptSource.SystemScriptGraph.NiagaraNodeAssignment_3.SetVariables_#5
 ]
}
emitters {
 Steam {
  emitterUpdateScript [
   &EmitterState(Life Cycle Mode=::NewEnumerator1, Loop Behavior=::NewEnumerator1, Loop Delay=0.1, Loop Duration=0.4, UseLoopDelay=true)
   &SpawnRate(SpawnRate=5, Use Spawn Probability=false)
   &SpawnPerUnit(Max Movement Threshold=100, Module.Use Max Movement Threshold=true, Module.Use Movement Tolerance=false, Use Spawn Probability=false)
  ]
  particleSpawnScript [
   &InitializeParticle(
    Color=<rgba 1 1 1 0.2>
    Lifetime Max=0.9
    Lifetime Min=0.6
    Lifetime Mode=::NewEnumerator1
    Position Offset=<33 0 0>
    Sprite Rotation Mode=::NewEnumerator2
    Sprite Size Mode=::NewEnumerator1
    Uniform Sprite Size Max=130
    Uniform Sprite Size Min=100
    UsePositionOffset=true
    Write Material Random=true
   )
   &AddVelocity(Velocity=&RandomRangeVector(Minimum=<50 1 30>, Maximum=<30 7 1>, Evaluation Type=::NewEnumerator1))
   &InheritVelocity(Module.Limit Speed=false, Velocity Limit=8, Velocity Scale=<0.25 0.25 15>) {enabled=false}
  ]
  particleUpdateScript [
   &ParticleState()
   /SetVariables(Transient.DistanceBeforeFade=650) {enabled=false, isSetParameters=true, module=SetVariables_#6}
   /SetVariables(Transient.DistanceFromCup:NiagaraFloat=&OneMinusFloat(Float=&ClampFloat(Float=&Multiply_Float(
       A=&Subtract_Float(
        A=&VectorLength(Vector=&Subtract_Vector(
          A=Linked() {linkedVariable {name=Particles.Position, type=NiagaraPosition}}
          B=Linked() {linkedVariable {name=Engine.Owner.Position, type=NiagaraPosition}}
         ))
        B=$Transient.DistanceBeforeFade
       )
       B=0.025
      )))) {enabled=false, isSetParameters=true, module=SetVariables_#7}
   &SubUVAnimation(
    Sub UV Play Rate:NiagaraFloat=&RandomRangeFloat(Minimum=0.05, Maximum=0.1)
    SubUV Animation Mode:ENiagara_SubUVLookupMode=::NewEnumerator3
    SubUV Start Time Offset:NiagaraFloat=&RandomRangeFloat(Minimum=0, Maximum=63)
   ) {enabled=false}
   &DynamicMaterialParameters(
    Index 0 Param 1=&FloatFromCurve(FloatCurve=DataInterface() {propertyValues {Curve {keys [<at 0 -0.25>,<at 10 0>]}}})
    Index 0 Param 4=&OneMinusFloat(Float=$Particles.NormalizedAge)
    Module.Param0WriteEnabled=true
    Module.Param1WriteEnabled=false
    Module.Param2WriteEnabled=false
    Module.Param3WriteEnabled=true
   ) {enabled=false}
   &ScaleSpriteSize(Scale Factor=&Vector2DFromFloat(Value=&FloatFromCurve(FloatCurve=DataInterface() {propertyValues {Curve {keys [
          {arriveTangent=0.4736, interpMode=RCIM_Cubic, leaveTangent=0.4736, tangentMode=RCTM_User, time=0.0011, value=0.2272}
          {arriveTangent=0.0192, interpMode=RCIM_Cubic, leaveTangent=0.0192, tangentMode=RCTM_User, time=1, value=0.4278}
         ]}}})))
   &Drag(
    Drag=&FloatFromCurve(FloatCurve=DataInterface() {propertyValues {Curve {keys [<at 0.0038 0.1198>,<at 0.4 1.875>,<at 1 2.375>]}}}, Scale Curve=1.5)
    Use Linear Drag=true
    Use Rotational Drag=false
   )
   &InheritVelocity(
    Module.Limit Speed=false
    Velocity Limit=0.3
    Velocity Scale=&VectorFromFloat(Value=0.1)
    Velocity Source=&RandomRangeVector(Minimum=<0 0 4>, Maximum=<1 1 15>)
   ) {enabled=false, module=InheritVelocity001}
   &SolveForcesAndVelocity(Clamp Velocity=false, Limit Acceleration=false)
   &DynamicMaterialParameters(
    Index 0 Param 1=&FloatFromCurve(FloatCurve=DataInterface() {propertyValues {Curve {defaultValue=1, keys [{tangentMode=RCTM_User, time=0, value=0}, {tangentMode=RCTM_User, time=1, value=1}]}}})
    Index 0 Param 2=&RandomRangeFloat(Minimum=0, Maximum=2)
    Index 0 Param 3=1
    Index 0 Param 4=1
    Index 1 Param 1=&FloatFromCurve(FloatCurve=DataInterface() {propertyValues {Curve {keys [
         {arriveTangent=-2.3163, interpMode=RCIM_Cubic, time=0, value=2}
         {arriveTangent=-1.9019, interpMode=RCIM_Cubic, leaveTangent=-1.9019, tangentMode=RCTM_User, time=1, value=0.4}
        ]}}})
    Index 1 Param 2=0.4
    Index 1 Param 3=0.6
    Index 1 Param 4=1
    Index 2 Param 1=&FloatFromCurve(FloatCurve=DataInterface() {propertyValues {Curve {keys [<at 0.46 0>,<at 1 1>]}}})
    Index 2 Param 2=45
    Module.Param0WriteEnabled=true
    Module.Param0WriteEnabled1=true
    Module.Param0WriteEnabled2=true
    Module.Param1WriteEnabled=true
    Module.Param1WriteEnabled1=true
    Module.Param1WriteEnabled2=true
    Module.Param2WriteEnabled=true
    Module.Param2WriteEnabled1=true
    Module.Param2WriteEnabled2=true
    Module.Param3WriteEnabled=true
    Module.Param3WriteEnabled1=true
    Module.Param3WriteEnabled2=true
    Write Parameter Index 1=true
    Write Parameter Index 2=true
   ) {module=DynamicMaterialParameters001}
   &ScaleColor(
    Linear Color Curve=DataInterface() {propertyValues {
      AlphaCurve {keys [<at 0 0>,<at 0.1809 8>,<at 0.7494 0.4>,<at 1 0>]}
      BlueCurve {keys [<at 0 0.1323>,<at 0.7558 0.0767>]}
      GreenCurve {keys [<at 0 0.1536>,<at 0.7558 0.0773>]}
      RedCurve {keys [<at 0 0.2073>,<at 0.7558 0.0906>]}
     }}
    Scale Mode=::NewEnumerator2
   ) {module=ScaleColor001}
   &ScaleColor(
    Scale RGB=&MakeVectorFromLinearColorRGB(Color=&FNTimeOfDayColor(
      Night Color=<rgba 0.1599 0.3151 0.6823 1>
      Morning Color=<rgba 1 0.8541 0.6031 1>
      Day Color=<rgba 1 1 1 1>
      Evening Color=<rgba 0.9311 0.8191 1 1>
     ))
    ScaleA=false
    ScaleRGB=true
   )
  ]
  rendererClasses [&NiagaraSpriteRendererProperties]
  renderers [{rendererClass=&NiagaraSpriteRendererProperties}]
 }
 Steam003 {
  emitterUpdateScript [
   &EmitterState(Life Cycle Mode=::NewEnumerator1, Loop Behavior=::NewEnumerator1, Loop Delay=0.2, Loop Duration=0.4, UseLoopDelay=true)
   &SpawnRate(SpawnRate=8, Use Spawn Probability=false)
   &SpawnPerUnit(Max Movement Threshold=100, Module.Use Max Movement Threshold=true, Module.Use Movement Tolerance=false, Use Spawn Probability=false)
  ]
  particleSpawnScript [
   &InitializeParticle(
    Color=<rgba 1 1 1 0.2>
    Lifetime Max=1.3
    Lifetime Mode=::NewEnumerator1
    Position Offset=<-60 -12 0>
    Sprite Rotation Mode=::NewEnumerator2
    Sprite Size Mode=::NewEnumerator1
    Uniform Sprite Size Max=130
    Uniform Sprite Size Min=100
    UsePositionOffset=true
    Write Material Random=true
   )
   &AddVelocity(Velocity=&RandomRangeVector(Minimum=<65 1 30>, Maximum=<87 7 1>, Evaluation Type=::NewEnumerator1))
   &InheritVelocity(Module.Limit Speed=false, Velocity Limit=8, Velocity Scale=<0.25 0.25 15>) {enabled=false}
  ]
  particleUpdateScript [
   &ParticleState()
   /SetVariables(Transient.DistanceBeforeFade=650) {enabled=false, isSetParameters=true, module=SetVariables_#6}
   /SetVariables(Transient.DistanceFromCup:NiagaraFloat=&OneMinusFloat(Float=&ClampFloat(Float=&Multiply_Float(
       A=&Subtract_Float(
        A=&VectorLength(Vector=&Subtract_Vector(
          A=Linked() {linkedVariable {name=Particles.Position, type=NiagaraPosition}}
          B=Linked() {linkedVariable {name=Engine.Owner.Position, type=NiagaraPosition}}
         ))
        B=$Transient.DistanceBeforeFade
       )
       B=0.025
      )))) {enabled=false, isSetParameters=true, module=SetVariables_#7}
   &SubUVAnimation(
    Sub UV Play Rate:NiagaraFloat=&RandomRangeFloat(Minimum=0.05, Maximum=0.1)
    SubUV Animation Mode:ENiagara_SubUVLookupMode=::NewEnumerator3
    SubUV Start Time Offset:NiagaraFloat=&RandomRangeFloat(Minimum=0, Maximum=63)
   ) {enabled=false}
   &DynamicMaterialParameters(
    Index 0 Param 1=&FloatFromCurve(FloatCurve=DataInterface() {propertyValues {Curve {keys [<at 0 -0.25>,<at 10 0>]}}})
    Index 0 Param 4=&OneMinusFloat(Float=$Particles.NormalizedAge)
    Module.Param0WriteEnabled=true
    Module.Param1WriteEnabled=false
    Module.Param2WriteEnabled=false
    Module.Param3WriteEnabled=true
   ) {enabled=false}
   &ScaleSpriteSize(Scale Factor=&Vector2DFromFloat(Value=&FloatFromCurve(FloatCurve=DataInterface() {propertyValues {Curve {keys [
          {arriveTangent=0.4736, interpMode=RCIM_Cubic, leaveTangent=0.4736, tangentMode=RCTM_User, time=0, value=0.25}
          {arriveTangent=0.3898, interpMode=RCIM_Cubic, leaveTangent=0.3898, time=0.221, value=0.2941}
          {arriveTangent=0.3743, interpMode=RCIM_Cubic, leaveTangent=0.3743, time=0.5091, value=0.4484}
          {arriveTangent=0.9326, interpMode=RCIM_Cubic, leaveTangent=0.9326, time=0.7133, value=0.4784}
          {arriveTangent=0.2329, interpMode=RCIM_Cubic, leaveTangent=0.2329, tangentMode=RCTM_User, time=0.9994, value=0.9057}
         ]}}})))
   &Drag(
    Drag=&FloatFromCurve(FloatCurve=DataInterface() {propertyValues {Curve {keys [<at 0.0038 0.1198>,<at 0.4 1.625>,<at 1 2>]}}})
    Use Linear Drag=true
    Use Rotational Drag=false
   )
   &InheritVelocity(
    Module.Limit Speed=false
    Velocity Limit=0.3
    Velocity Scale=&VectorFromFloat(Value=0.1)
    Velocity Source=&RandomRangeVector(Minimum=<0 0 4>, Maximum=<1 1 15>)
   ) {enabled=false, module=InheritVelocity001}
   &SolveForcesAndVelocity(Clamp Velocity=false, Limit Acceleration=false)
   &DynamicMaterialParameters(
    Index 0 Param 1=&FloatFromCurve(FloatCurve=DataInterface() {propertyValues {Curve {defaultValue=1, keys [{tangentMode=RCTM_User, time=0, value=0}, {tangentMode=RCTM_User, time=1, value=1}]}}})
    Index 0 Param 2=&RandomRangeFloat(Minimum=0, Maximum=2)
    Index 0 Param 3=1
    Index 0 Param 4=1
    Index 1 Param 1=&FloatFromCurve(FloatCurve=DataInterface() {propertyValues {Curve {keys [
         {arriveTangent=-2.3163, interpMode=RCIM_Cubic, time=0, value=2}
         {arriveTangent=-1.9019, interpMode=RCIM_Cubic, leaveTangent=-1.9019, tangentMode=RCTM_User, time=1, value=0.4}
        ]}}})
    Index 1 Param 2=0.4
    Index 1 Param 3=0.6
    Index 1 Param 4=1
    Index 2 Param 1=&FloatFromCurve(FloatCurve=DataInterface() {propertyValues {Curve {keys [<at 0.46 0>,<at 1 1>]}}})
    Index 2 Param 2=45
    Module.Param0WriteEnabled=true
    Module.Param0WriteEnabled1=true
    Module.Param0WriteEnabled2=true
    Module.Param1WriteEnabled=true
    Module.Param1WriteEnabled1=true
    Module.Param1WriteEnabled2=true
    Module.Param2WriteEnabled=true
    Module.Param2WriteEnabled1=true
    Module.Param2WriteEnabled2=true
    Module.Param3WriteEnabled=true
    Module.Param3WriteEnabled1=true
    Module.Param3WriteEnabled2=true
    Write Parameter Index 1=true
    Write Parameter Index 2=true
   ) {module=DynamicMaterialParameters001}
   &ScaleColor(
    Linear Color Curve=DataInterface() {propertyValues {
      AlphaCurve {keys [<at 0 0>,<at 0.1809 8>,<at 0.7494 0.4>,<at 1 0>]}
      BlueCurve {keys [<at 0 0.1333>,<at 0.7558 0.0626>]}
      GreenCurve {keys [<at 0 0.1648>,<at 0.7558 0.0631>]}
      RedCurve {keys [<at 0 0.2073>,<at 0.7558 0.074>]}
     }}
    Scale Mode=::NewEnumerator2
   ) {module=ScaleColor001}
   &ScaleColor(
    Scale RGB=&MakeVectorFromLinearColorRGB(Color=&FNTimeOfDayColor(
      Night Color=<rgba 0.157 0.3094 0.6698 1>
      Morning Color=<rgba 1 0.7536 0.6061 1>
      Day Color=<rgba 1 1 1 1>
      Evening Color=<rgba 0.9352 0.835 1 1>
     ))
    ScaleA=false
    ScaleRGB=true
   )
  ]
  rendererClasses [&NiagaraSpriteRendererProperties]
  renderers [{rendererClass=&NiagaraSpriteRendererProperties}]
 }
}
summary {
 systemName=NS_ArmyFlour_Emote_CrispRover
 emitters [
  {emitterName=Steam, rendererClasses [&NiagaraSpriteRendererProperties]}
  {emitterName=Steam003, rendererClasses [&NiagaraSpriteRendererProperties]}
 ]
 userVariables []
}
