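# Niagara smoke-trail effect for the ArmyFlour player emote, drawing short-lived ribbons from points on a skeletal mesh.
# The active emitter spawns four ribbons, mapping spawn groups to bone/socket indices, then pushes them upward in world space while following source movement.
# Curves shift color from warm yellow toward pale cream and fade alpha over life.
# A disabled variant emitter adds curl noise and a higher spawn rate.
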NiagaraSystem /BRCosmetics/Animation/Game/MainPlayer/Emotes/ArmyFlour/FX/NS_ArmyFlour_SmokeTrails
dependencies {
 usedDataInterfaces [&NiagaraDataInterfaceCurve, &NiagaraDataInterfaceColorCurve, &NiagaraDataInterfaceSkeletalMesh]
 usedDynamicInputs []
 usedModules [
  &SampleSkeletalMesh
  &self:SystemSpawnScript.SystemScriptSource.SystemScriptGraph.NiagaraNodeAssignment_0.SetVariables_#1
  &SystemState
 ]
}
emitters {
 FN_Ribbons {
  bEnabled=false
  emitterUpdateScript [
   &EmitterState()
   /Game/Effects/Niagara/Modules/Skydiving/FN_DivingAttributes/FN_EmitterDivingAttributes_v1(Accumulate Dive=0, SkeletalMesh:NiagaraDataInterfaceSkeletalMesh=$User.SkeletalMesh) {enabled=false}
   &FN_SpawnRibbon_v2(Number of Ribbons=4, Spawn Mode=::NewEnumerator0, SpawnRate=90)
  ]
  particleSpawnScript [
   &SystemLocation(MaskSpawnGroup=false, Offset Coordinate Space=::Simulation)
   &FN_InitializeRibbon_v3(
    Color:LinearColor=/Niagara/DynamicInputs/TypeConversions/MakeLinearColorFromFloat(Value:NiagaraFloat=12)
    Lifetime=0.35
    Ribbon Width:NiagaraFloat=&FloatFromCurve(
     FloatCurve=DataInterface() {propertyValues {Curve {keys [
         <at 0 1 RCIM_Cubic>
         {arriveTangent=-0.5895, interpMode=RCIM_Cubic, leaveTangent=-0.5895, tangentMode=RCTM_User, time=0.2, value=0.5}
         <at 1 0.125 RCIM_Cubic>
        ]}}}
     Scale Curve=10
    )
   ) {module=FN_InitializeRibbon_v3001}
   &SkeletalMeshLocation(
    Bone / Socket Index=&MaskIntBySpawnGroup(
     Masked Int=0
     Passthrough Int=&MaskIntBySpawnGroup(
      Masked Int=1
      Passthrough Int=&MaskIntBySpawnGroup(
       Masked Int=2
       Passthrough Int=&MaskIntBySpawnGroup(Masked Int=3, Passthrough Int=1, Spawn Group Mask=3)
       Spawn Group Mask=2
      )
      Spawn Group Mask=1
     )
     Spawn Group Mask=0
    )
    Bone Sampling Mode=::NewEnumerator4
    MaskSpawnGroup=false
    Skeletal Mesh=$User.SkeletalMesh
   )
   &AddVelocity(
    Coordinate Space=::World
    Scale Added Velocity=&Lerp_Vector(A=<1 1 1>, B=<1 1 2>, Alpha=$Emitter.Dive)
    Velocity=<0 0 95>
   )
   &AddVelocity(Velocity=&MakeCustomVectorFromBool(
     Bool=$Particles.Locked
     True Vector=<0 0 0>
     False Vector=&FN_FrontEndVector3Modifier(Gameplay=<0 0 83>, Front End=<0 0 83>)
    )) {module=AddVelocity001}
   &SampleSkeletalMeshSkeleton(
    Mesh:NiagaraDataInterfaceSkeletalMesh=$User.SkeletalMesh
    Sampled Position Coordinate Space:ENiagaraCoordinateSpace=::Local
    Sampled Rotation Coordinate Space:ENiagaraCoordinateSpace=::Local
    Sampled Velocity Coordinate Space:ENiagaraCoordinateSpace=::Local
   ) {enabled=false}
   &AddVelocity(Velocity=&Multiply_VectorByFloat(Vector=<0 0 100>, Float=&MaskFloatBySpawnGroup(Masked Float=5, Passthrough Float=0))) {enabled=false, module=AddVelocity002}
  ]
  particleUpdateScript [
   &CurlNoiseForce(
    Noise Frequency=33
    Noise Strength=&FloatFromCurve(FloatCurve=DataInterface() {propertyValues {Curve {keys [<at 0 0.5 RCIM_Cubic>,<at 0.5 0.85 RCIM_Cubic>]}}}, Scale Curve=100)
    Pan Noise=false
   )
   &InheritVelocity(Inherited Velocity Speed Limit=200, Speed Limit=true, UseSpeedThreshold=false) {enabled=false}
   &SolveForcesAndVelocity(Clamp Velocity=false, Limit Acceleration=false)
   &FN_UpdateRibbon(Initial Color:LinearColor=$Particles.Initial.Color, Initial RibbonWidth:NiagaraFloat=$Particles.Initial.RibbonWidth)
   &Color(
    Color=&ColorFromCurve(ColorCurve=DataInterface() {propertyValues {
       AlphaCurve {keys [<at 0.0961 0>,<at 0.2505 12>,<at 0.4544 7.4458>,<at 0.9294 0>]}
       BlueCurve {keys [
         <at 0 0.172>
         <at 0.1607 0.0359>
         {arriveTangent=1.0349, interpMode=RCIM_Cubic, leaveTangent=1.0349, time=0.3044, value=0.5469}
         <at 0.4709 0.6594>
        ]}
       GreenCurve {keys [
         <at 0 0.749>
         <at 0.1607 0.6179>
         {arriveTangent=0.111, interpMode=RCIM_Cubic, leaveTangent=0.111, time=0.3044, value=0.8112}
         <at 0.4709 0.8013>
        ]}
       RedCurve {keys [<at 0 1>,<at 0.1607 1>,<at 0.3044 1 RCIM_Cubic>,<at 0.4709 1>]}
      }})
    Scale Alpha=&FloatFromCurve(FloatCurve=DataInterface() {propertyValues {Curve {keys [
         {arriveTangent=0.7543, interpMode=RCIM_Cubic, leaveTangent=0.7543, tangentMode=RCTM_User, time=0, value=0}
         <at 0.1 0.05>
         {arriveTangent=-0.013, interpMode=RCIM_Cubic, leaveTangent=-0.013, tangentMode=RCTM_User, time=0.2992, value=0.0096}
         {interpMode=RCIM_Cubic, leaveTangent=-0.2611, time=1, value=0}
        ]}}})
   )
   &Material(Base Texture Time:NiagaraFloat=&Multiply_Float(A=$Emitter.Age, B=-1), PostMult=0.3971, Power=0.6448)
  ]
  rendererClasses [&NiagaraRibbonRendererProperties]
  renderers [{rendererClass=&NiagaraRibbonRendererProperties}]
 }
 FN_Ribbons005 {
  emitterUpdateScript [
   &EmitterState()
   &FN_SpawnRibbon_v2(Number of Ribbons=4, Spawn Mode=::NewEnumerator0, SpawnRate=22)
  ]
  particleSpawnScript [
   &Material(Base Texture Time:NiagaraFloat=&Multiply_Float(A=$Emitter.Age, B=-1), PostMult=0.3971, Power=0.6448)
   &SystemLocation(MaskSpawnGroup=false, Offset Coordinate Space=::Simulation)
   &FN_InitializeRibbon_v3(
    Color:LinearColor=&MakeLinearColorFromFloat(Value:NiagaraFloat=12)
    Lifetime=0.35
    Ribbon Width:NiagaraFloat=&FloatFromCurve(
     FloatCurve=DataInterface() {propertyValues {Curve {keys [
         <at 0 1 RCIM_Cubic>
         {arriveTangent=-0.5895, interpMode=RCIM_Cubic, leaveTangent=-0.5895, tangentMode=RCTM_User, time=0.2, value=0.5}
         <at 1 0.125 RCIM_Cubic>
        ]}}}
     Scale Curve=10
    )
   ) {module=FN_InitializeRibbon_v3001}
   &SkeletalMeshLocation(
    Bone / Socket Index=&MaskIntBySpawnGroup(
     Masked Int=0
     Passthrough Int=&MaskIntBySpawnGroup(
      Masked Int=1
      Passthrough Int=&MaskIntBySpawnGroup(
       Masked Int=2
       Passthrough Int=&MaskIntBySpawnGroup(Masked Int=3, Passthrough Int=1, Spawn Group Mask=3)
       Spawn Group Mask=2
      )
      Spawn Group Mask=1
     )
     Spawn Group Mask=0
    )
    Bone Sampling Mode=::NewEnumerator4
    MaskSpawnGroup=false
    Skeletal Mesh=$System.Skeletal Mesh
   )
   &AddVelocity(
    Coordinate Space=::World
    Scale Added Velocity=&Lerp_Vector(A=<1 1 1>, B=<1 1 2>, Alpha=$Emitter.Dive)
    Velocity=<0 0 155>
   )
  ]
  particleUpdateScript [
   &Drag(Drag=-7, Use Linear Drag=true, Use Rotational Drag=true) {enabled=false}
   &InheritSourceMovement(ApplyForce=false, ApplyPosition=true, ApplyPositionFalloff=false)
   &SolveForcesAndVelocity(Clamp Velocity=false, Limit Acceleration=false)
   &FN_UpdateRibbon(Initial Color:LinearColor=$Particles.Initial.Color, Initial RibbonWidth:NiagaraFloat=$Particles.Initial.RibbonWidth)
   &Color(
    Color=&ColorFromCurve(ColorCurve=DataInterface() {propertyValues {
       AlphaCurve {keys [<at 0.0961 0>,<at 0.2505 12>,<at 0.4544 7.4458>,<at 0.9294 0>]}
       BlueCurve {keys [
         <at 0 0.172>
         <at 0.1607 0.0359>
         {arriveTangent=1.0349, interpMode=RCIM_Cubic, leaveTangent=1.0349, time=0.3044, value=0.5469}
         <at 0.4709 0.6594>
        ]}
       GreenCurve {keys [
         <at 0 0.749>
         <at 0.1607 0.6179>
         {arriveTangent=0.111, interpMode=RCIM_Cubic, leaveTangent=0.111, time=0.3044, value=0.8112}
         <at 0.4709 0.8013>
        ]}
       RedCurve {keys [<at 0 1>,<at 0.1607 1>,<at 0.3044 1 RCIM_Cubic>,<at 0.4709 1>]}
      }})
    Scale Alpha=&FloatFromCurve(FloatCurve=DataInterface() {propertyValues {Curve {keys [
         {arriveTangent=0.7543, interpMode=RCIM_Cubic, leaveTangent=0.7543, tangentMode=RCTM_User, time=0, value=0}
         <at 0.1 0.05>
         {arriveTangent=-0.013, interpMode=RCIM_Cubic, leaveTangent=-0.013, tangentMode=RCTM_User, time=0.2992, value=0.0096}
         {interpMode=RCIM_Cubic, leaveTangent=-0.2611, time=1, value=0}
        ]}}})
   )
  ]
  rendererClasses [&NiagaraRibbonRendererProperties]
  renderers [{rendererClass=&NiagaraRibbonRendererProperties}]
 }
}
summary {
 systemName=NS_ArmyFlour_SmokeTrails
 emitters [
  {emitterName=FN_Ribbons005, rendererClasses [&NiagaraRibbonRendererProperties]}
  {bEnabled=false, emitterName=FN_Ribbons, rendererClasses [&NiagaraRibbonRendererProperties]}
 ]
 userVariables [
  User.bIsFrontEnd {type=NiagaraBool}
  User.bIsFrontEndPreview {type=NiagaraBool}
  User.RibbonIDOffset {type=NiagaraInt32}
 ]
}
